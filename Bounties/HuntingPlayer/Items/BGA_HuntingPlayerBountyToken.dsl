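; Bounty token pickup for Hunting Player that only the owning team sees and can collect.
; On BeginPlay it traces down on authority for a stop location, sets its team, starts a lifetime timer and reveals itself through OnRep_ShouldShow.
; Collecting plays effects on local teammates, then the server sets replicated HideAndKill to hide and expire the actor.
; Lifetime expiry blocks new interactions and waits one interaction duration before destroying it.
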(event EventBeginPlay
  ;[0 0 0] BeginPlay
  (CallFunction|DetermineStopLocation)
  (CallFunction|SetTeam (Variables|Default|GetOwnerTeam))
  (Variables|Default|SetShouldShow true)
  (Utilities|Time|SetTimerbyEvent (EventDispatchers|CreateEvent :!event Custom|OnLifetimeExpired) (Variables|Default|GetLifetime)))

(event Interaction|EventBlueprintOnInteract
  (params (InteractingPawn :type Fort%20Pawn%20Object%20Reference) (InteractionBeingAttempted :type EInteractionBeingAttempted%20Enum :default "FirstInteraction"))
  ;[0 0 0] OnInteract
  (CallFunction|CollectPickup))

(event Custom|CollectPickup
  ;[0 0 0] CollectPickup
  (Utilities|FlowControl|Sequence
    (:then_0
      (for _array_element (Fortnite|GetLocalFortPlayerControllers true)
        (if (Math|Byte|Equal(Byte) (Team|GetActorTeam _array_element) (Variables|Default|GetOwnerTeam))
          (Utilities|IsValid (Game|Player|GetPlayerPawn _array_element)
            (:Is%20Valid
              (Effects|SpawnEmitterAttached :EmitterTemplate (Variables|Default|GetCollected_PS) :AttachToComponent (Variables|/Script/Engine.Character|GetMesh (Game|Player|GetPlayerPawn _array_element)) :AttachPointName "root" :Scale "1.000000,1.000000,1.000000" :LocationType "SnapToTarget" :bAutoDestroy true :PoolingMethod "None" :bAutoActivate true) ; Each squadmate sees a card pop above their head
              (Audio|PlaySound2D (Variables|Default|GetPickUpSound)))))))
    (:then_1
      (CallFunction|DestroyPickup))))

(event Custom|DestroyPickup
  ;[0 0 0] DestroyPickup
  (if (and (Networking|IsServer) (not (Variables|Default|GetHideAndKill)))
    (Variables|Default|SetHideAndKill true)))

(event Custom|HideAndKillPickup
  ;[0 0 0] HideAndKillPickup
  (Niagara|SpawnSystematLocation "/Game/Athena/Items/EnvironmentalItems/SCMachine/FX/NS_SCCard_Collect.NS_SCCard_Collect" (Transformation|GetWorldLocation (Variables|Default|GetToken)))
  (Audio|Components|Audio|FadeOut (Variables|Default|GetSC_Machine_Memory_Card_Loop_Cue) 0.15 0.0 "Linear" :!class /Script/Engine.AudioComponent)
  (Collision|SetActorEnableCollision false)
  (Rendering|SetActorHiddenInGame true)
  (Actor|SetLifeSpan (Math|Random|RandomFloatinRange 0.01 0.5)))

(event EventOnDeathServer
  (params (Damage :type Float%20%28single-precision%29) (DamageTags :type Gameplay%20Tag%20Container%20Structure%20%28by%20ref%29) (Momentum :type Vector :default "0, 0, 0") (HitInfo :type Hit%20Result%20Structure%20%28by%20ref%29) (InstigatedBy :type Controller%20Object%20Reference) (DamageCauser :type Actor%20Object%20Reference) (EffectContext :type Gameplay%20Effect%20Context%20Handle%20Structure))
  ;[0 0 0] OnDeathServer
  ; Set the is pending kill var when the chip is killed to make sure that if we kill it on disconncet it does not play sounds or visuals.
  (Variables|Default|SetIsPendingKill true))

(event Custom|OnLifetimeExpired
  ;[0 0 0] OnLifetimeExpired
  (Variables|Default|SetIsDelayingDeath true)
  (bind _outxy (Fortnite|EvaluateCurveTableRow (Variables|Interaction|GetInteractionSpeed)))
  (Utilities|FlowControl|Delay _outxy)
  (CallFunction|DestroyPickup))

(fn BlueprintCanInteract
  (params (InteractingPawn :type Fort%20Pawn%20Object%20Reference) (InteractionBeingAttempted :type EInteractionBeingAttempted%20Enum :default "FirstInteraction") (InteractionType :type TInteractionType%20Enum :default "IT_NoInteraction"))
  (return (and (Math|Byte|Equal(Byte) (Team|GetActorTeam InteractingPawn) (Variables|Default|GetOwnerTeam)) (not (Variables|Default|GetIsDelayingDeath)))))

(fn BlueprintGetInteractionString
  (params (InteractingPawn :type Fort%20Pawn%20Object%20Reference) (InteractionBeingAttempted :type EInteractionBeingAttempted%20Enum :default "FirstInteraction"))
  (return (Variables|Default|GetInteractString)))

(fn DetermineStopLocation
  (Networking|SwitchHasAuthority
    (:Authority
      (bind (_outhit _returnvalue) (Collision|CapsuleTraceByChannel (Transformation|GetActorLocation) (- (Transformation|GetActorLocation) "0, 0,10000.000000") (Variables|/Script/Engine.CapsuleComponent|GetCapsuleRadius (Variables|Default|GetCollision)) (Variables|/Script/Engine.CapsuleComponent|GetCapsuleHalfHeight (Variables|Default|GetCollision)) "TraceTypeQuery10" false (Utilities|Array|MakeArray)))
      (if _returnvalue
        (Variables|Default|SetStopLocation (.Location (Collision|BreakHitResult _outhit)))
        (else
          (Fortnite|FortShippingLog self "BGA_HuntingPlayerBountyToken - Failed to find ground to stop at!"))))))

(fn OnRep_HideAndKill
  (if (Variables|Default|GetHideAndKill)
    (CallFunction|HideandKillPickup)))

(fn OnRep_ShouldShow
  (bind _shouldshow (Variables|Default|GetShouldShow))
  (Variables|Interaction|SetbAllowInteract true)
  (Rendering|SetVisibility (Variables|Default|GetCollision) (Variables|Default|GetShouldShow) true)
  (for _array_element (Fortnite|GetLocalFortPlayerControllers true)
    (if (Math|Byte|Equal(Byte) (Variables|Default|GetOwnerTeam) (Team|GetActorTeam _array_element))
      ;[1 1 1] Set The indicator color and turn on the glow if the player is on the same team
      (Development|SetHiddeninGame (Variables|Default|GetBackgroundGlow) (not _shouldshow) false)
      (Components|Activation|SetActive (Variables|Default|GetSpawnInEffect) _shouldshow true)
      (Rendering|Material|SetScalarParameterValueonMaterials (Variables|Default|GetToken) "StartTime" (Utilities|Time|GetGameTimeinSeconds))
      (Audio|Components|Audio|FadeIn (Variables|Default|GetSC_Machine_Memory_Card_Loop_Cue) 0.25 1.0 0.0 "Linear" :!class /Script/Engine.AudioComponent)
      (else
        ;[1 1 1] Hide actor if not on same team and turn off its collision
        (Rendering|SetActorHiddenInGame true)
        (Collision|SetActorEnableCollision false)))))

(fn ConstructionScript)
